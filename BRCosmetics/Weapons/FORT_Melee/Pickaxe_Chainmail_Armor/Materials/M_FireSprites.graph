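# Unlit translucent Niagara sprite material that draws animated, wispy fire for the Chainmail Armor pickaxe cosmetic.
# Panning fire noise in polar UVs, distorted by blurred noise, is masked by a side-biased radial gradient and blended toward a center value, with particle dynamic parameters controlling the shape.
# The result and particle alpha drive a curve-atlas color ramp, tinted by particle color, and smoothstepped opacity, depth-faded on the default and deferred paths.
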Material /BRCosmetics/Weapons/FORT_Melee/Pickaxe_Chainmail_Armor/Materials/M_FireSprites
graph {
 properties {bUsedWithNiagaraSprites=true, blendMode=BLEND_Translucent, pixelDepthOffsetMode=PDOM_Legacy, shadingModel=MSM_Unlit, usedShadingModels=Unlit}
 nodes [
  n1 = ParticleColor()
  n2 = MaterialFunctionCall!(UV, Twist Center, Twist Edges) {materialFunction=/Game/Accessories/FORT_Backpacks/Backpack_F_MED_Celestial_Glow/FX/Materials/MF_PolarUV}
  n3 = Saturate(None=OneMinus(None=ComponentMask(None=@n2.Polar UVs) {g=true}))
  n4 = Subtract(A=ComponentMask(None=TextureCoordinate()) {r=true}) {constB=0.5}
  n5 = DynamicParameter() {defaultValue=<rgba 0 2 0 1>, paramNames [Gradient Bias, Power, Distortion, Center Value]}
  n6 = MaterialFunctionCall(Base=@n3, Blend=Add(A=Multiply(A=@n4, B=@n5.Gradient Bias) {constB=0.5}) {constB=0.5}) {materialFunction=&Blend_Overlay}
  n7 = Panner(Coordinate=Multiply(A=TextureCoordinate()) {constB=0.6}) {speedX=0.1, speedY=0.1}
  n8 = Multiply(
   A=Subtract(A=TextureSample(UVs=@n7) {samplerType=SAMPLERTYPE_Grayscale, texture=/Game/UI/Frontend/ItemShop/Art/Rumble/LowResBlurredNoise_Rumble}.R) {constB=0.5}
   B=DynamicParameter() {defaultValue=<rgba 0 2 0 1>, paramNames [Gradient Bias, Power, Distortion, Center Value]}.Distortion
  ) {constB=0}
  n9 = ScalarParameter() {defaultValue=1, parameterName=Scale X, sliderMax=5}
  n10 = ScalarParameter() {defaultValue=1, parameterName=Scale Y, sliderMax=5}
  n11 = Panner(
   Coordinate=Multiply(A=Add(A=@n2.Polar UVs, B=@n8), B=AppendVector(A=@n9, B=@n10)) {constB=2}
   Speed=AppendVector(A=ScalarParameter() {parameterName=Speed X, sliderMax=5}, B=ScalarParameter() {defaultValue=-0.5, parameterName=Speed Y, sliderMax=5})
  ) {speedY=-0.5}
  n12 = DynamicParameter() {defaultValue=<rgba 0 2 0 1>, paramNames [Gradient Bias, Power, Distortion, Center Value]}
  n13 = LinearInterpolate(
   A=Multiply(
    A=ComponentMask(None=@n6.Result) {r=true}
    B=TextureSample!(UVs=@n11, Tex, "DDX(UVs)"=@n2.DDX, "DDY(UVs)"=@n2.DDY, Apply View MipBias) {mipValueMode=TMVM_Derivative, texture=/Game/Effects/Fort_Effects/Textures/T_WispyFireNoise03}.R
   )
   B=@n12.Center Value
   Alpha=SmoothStep(Value=Power(Base=@n3, Exp=@n12.Power))
  )
  n14 = Power(Base=@n13, Exp=ScalarParameter() {defaultValue=1, parameterName=Color Power, sliderMax=5})
  n15 = CurveAtlasRowParameter(CurveTime=SmoothStep(
    Min=ScalarParameter() {defaultValue=0.3, parameterName=Color SmoothStep Min, sliderMax=3, sliderMin=-3}
    Max=ScalarParameter() {defaultValue=1.2, parameterName=Color SmoothStep Max, sliderMax=3, sliderMin=-3}
    Value=Multiply(A=@n1.A, B=@n14)
   ) {constMax=1.2, constMin=0.3}) {
   atlas=/Game/Athena/LinearColorCurves/CurveAtlas_Chainmail
   curve=/Game/Athena/LinearColorCurves/Curve_ChainMail_GradientForMesh
   defaultValue=2
  }
  MaterialFunctionCall(Emissive=Multiply(A=@n1.RGB, B=@n15)) {materialFunction=&MF_FX_EmissiveAdjustments}
  n16 = Multiply(A=@n13, B=@n1.A)
  n17 = DepthFade(Opacity=@n16, FadeDistance=ScalarParameter() {defaultValue=35, parameterName=Depth Fade Distance, sliderMax=100})
  SmoothStep(Value=ShadingPathSwitch(Default=@n17, Deferred=@n17, Forward=@n16, Mobile=@n16)) {constMax=0.5}
 ]
}
