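# Translucent, unlit, two-sided Niagara material for the idle flame bowl effect on the Chainmail Armor pickaxe.
# Two multiplied panning Voronoi samples distort wispy noise, which carves a vertical gradient faded by view angle and camera depth.
# That mask drives curve-atlas color with time-of-day emissive scaling, plus smoothstepped opacity, depth-faded on deferred paths.
# Dynamic parameters feed U offset, dive brightness, camera offset and speed-based bend, alongside sine-wave offset math.
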Material /BRCosmetics/Weapons/FORT_Melee/Pickaxe_Chainmail_Armor/Materials/M_Pickaxe_ChainmailArmor_IdleFlameBowl
graph {
 properties {
  bUsedWithNiagaraMeshParticles=true
  bUsedWithNiagaraRibbons=true
  bUsedWithNiagaraSprites=true
  blendMode=BLEND_Translucent
  pixelDepthOffsetMode=PDOM_Legacy
  shadingModel=MSM_Unlit
  twoSided=true
  usedShadingModels=Unlit
 }
 nodes [
  n1 = DynamicParameter() {parameterIndex=1, paramNames [Speed X, Speed Y, Speed Z, Param4]}
  n2 = AppendVector(A=AppendVector(A=@n1.Speed X, B=@n1.Speed X), B=@n1.Speed Z)
  n3 = ComponentMask(None=TextureCoordinate()) {g=true}
  n4 = Multiply(
   A=Add(A=Constant3Vector() {constant=<rgba 0 0 1 0>}, B=Multiply(A=@n2) {constB=-0.005})
   B=Power(Base=OneMinus(None=@n3), Exp=ScalarParameter() {defaultValue=1, parameterName=WPO Bend, sliderMax=10})
  )
  Add(
   A=Multiply(
    A=MaterialFunctionCall() {materialFunction=&CameraDirectionVector}.Result
    B=DynamicParameter() {defaultValue=<rgba 0 1 0 1>, paramNames [UV1 U Offset, Dive, Camera Offset, Param4]}.Camera Offset
   )
   B=@n4
  )
  MaterialFunctionCall() {materialFunction=&FadeBasedOnViewAngle}
  n5 = Add(
   A=Multiply(A=TextureCoordinate(), B=ScalarParameter() {defaultValue=2, parameterName=Vertical Gradient Tiling, sliderMax=3})
   B=NamedRerouteUsage()
  )
  n6 = CurveAtlasRowParameter(CurveTime=Saturate(None=ComponentMask(None=@n5) {g=true})) {
   atlas=/Game/Athena/LinearColorCurves/CurveAtlas_Chainmail
   curve=/Game/Athena/LinearColorCurves/Curve_ChainMail_GradientForMesh
   defaultValue=2
   parameterName=Curve_Color
  }
  n7 = Multiply(A=Multiply(A=@n6, B=@n6), B=ScalarParameter() {parameterName=Vertical Gradient Bloom, sliderMax=1})
  n8 = TextureCoordinate()
  n9 = Power(
   Base=OneMinus(None=ComponentMask(None=@n8) {g=true})
   Exp=ScalarParameter() {defaultValue=2, group=Shape Lerp, parameterName=Gradient Power, sliderMax=10}
  )
  n10 = ScalarParameter() {defaultValue=2, group=Shape Lerp, parameterName=Gradient Intensity, sliderMax=10}
  n11 = NamedRerouteUsage()
  n12 = ScalarParameter() {group=Wispy Flame, parameterName=Wispy Distortion Strength, sliderMax=1, sliderMin=-1}
  n13 = Panner(
   Coordinate=Add(A=TextureCoordinate(), B=Multiply(A=@n11, B=@n12) {constB=0.1})
   Speed=AppendVector(
    A=ScalarParameter() {group=Wispy Flame, parameterName=Speed X, sliderMax=5, sliderMin=-5}
    B=ScalarParameter() {defaultValue=-1, group=Wispy Flame, parameterName=Speed Y, sliderMax=5, sliderMin=-5}
   )
  ) {speedX=1}
  n14 = AppendVector(
   A=ScalarParameter() {defaultValue=1, group=Wispy Flame, parameterName=Scale X, sliderMax=5}
   B=ScalarParameter() {defaultValue=0.5, group=Wispy Flame, parameterName=Scale Y, sliderMax=5}
  )
  n15 = DynamicParameter() {defaultValue=<rgba 0 1 0 1>, paramNames [UV1 U Offset, Dive, Camera Offset, Param4]}
  n16 = Constant()
  n17 = TextureSampleParameter2D(UVs=Add(A=Multiply(A=@n13, B=@n14), B=AppendVector(A=@n15.UV1 U Offset, B=@n16))) {
   group=Wispy Flame
   parameterName=Flame Texture
   texture=/BRCosmetics/Athena/Effects/Particles/Skydiving/NS_Contrail_Chainmail/T_WispyFireNoise03
  }
  n18 = ScalarParameter() {defaultValue=512, parameterName=Camera Fade Length, sliderMax=200}
  n19 = ScalarParameter() {defaultValue=24, parameterName=Camera Fade Offset, sliderMax=200}
  n20 = MaterialFunctionCall() {materialFunction=&CameraDirectionVector}
  n21 = Transform(None=Constant3Vector() {constant=<rgba -1 0 0 0>}) {transformSourceType=TRANSFORMSOURCE_Instance}
  n22 = Power(
   Base=Abs(None=DotProduct(A=@n20.Result, B=@n21))
   Exp=ScalarParameter() {defaultValue=8, parameterName=ViewBasedFadeSensitivity, sliderMax=16}
  )
  n23 = Multiply(
   A=Subtract(A=Multiply(A=@n9, B=@n10) {constB=3}, B=@n17.R)
   B=Multiply(
    A=MaterialFunctionCall(Fade Length=@n18, Fade Offset=@n19) {materialFunction=&CameraDepthFade}.Result
    B=Clamp(None=@n22)
   ) {constA=1}
  )
  n24 = SmoothStep(
   Min=ScalarParameter() {group=Color, parameterName=Color Smooth Step Min, sliderMax=1, sliderMin=-1}
   Max=ScalarParameter() {defaultValue=2, group=Color, parameterName=Color Smooth Step Max, sliderMax=5}
   Value=Power(Base=Saturate(None=@n23), Exp=ScalarParameter() {defaultValue=1, group=Color, parameterName=Color Power, sliderMax=5}) {constExponent=1}
  ) {constMax=2}
  n25 = CurveAtlasRowParameter(CurveTime=@n24) {
   atlas=&CurveAtlas_Chainmail
   curve=&Curve_ChainMail_GradientForMesh
   defaultValue=2
   parameterName=Curve_Color
  }
  n26 = LinearInterpolate(
   A=ScalarParameter() {defaultValue=1, parameterName=Emissive, sliderMax=2}
   B=ScalarParameter() {defaultValue=1.5, parameterName=Emissive_Dive, sliderMax=3}
   Alpha=DynamicParameter() {defaultValue=<rgba 0 1 0 1>, paramNames [UV1 U Offset, Dive, Camera Offset, Param4]}.Dive
  ) {constA=1}
  n27 = Multiply(A=@n25, B=Add(A=Multiply(A=@n24, B=@n26)) {constB=0.5})
  n28 = VectorParameter() {parameterName=TOD color, defaultValue=<rgba 1 1 1 0.5>}
  MaterialFunctionCall(Emissive=Multiply(
    A=Add(A=@n7, B=@n27)
    B=MaterialFunctionCall(Value_Morning=@n28.R, Value_Day=@n28.R, Value_Evening=@n28.R, Value_Night=@n28.R) {materialFunction=&Scalar_Time_of_Day}.Result
   )) {materialFunction=&MF_FX_EmissiveAdjustments}
  Multiply(A=ScalarParameter() {parameterName=WPO Intensity, sliderMax=100}, B=@n4)
  n29 = SmoothStep(Max=OneMinus(None=ScalarParameter() {defaultValue=0.9, parameterName=Edge Hardness, sliderMax=1}), Value=@n23) {constMax=0.1}
  n30 = DepthFade(Opacity=@n29, FadeDistance=ScalarParameter() {defaultValue=35, parameterName=Depth Fade Distance, sliderMax=100}) {fadeDistanceDefault=25}
  Multiply(A=Multiply(
    A=ShadingPathSwitch(Default=@n30, Deferred=@n30, Forward=@n29, Mobile=@n29)
    B=ScalarParameter() {defaultValue=1, parameterName=Opacity, sliderMax=1}
   ))
  n31 = Add(
   A=TextureCoordinate() {coordinateIndex=1}
   B=AppendVector(A=DynamicParameter() {defaultValue=<rgba 0 1 0 1>, paramNames [UV1 U Offset, Dive, Camera Offset, Param4]}.UV1 U Offset, B=Constant())
  )
  n32 = Sine(None=ComponentMask(None=Panner(Coordinate=@n31) {speedY=-1}) {g=true}) {period=1.4}
  n33 = Panner(Coordinate=@n31) {speedY=-0.5}
  n34 = Multiply(A=Sine(None=ComponentMask(None=@n33) {g=true}) {period=0.5}) {constB=0.3}
  n35 = Sine(None=ComponentMask(None=Panner(Coordinate=@n31) {speedY=-0.75}) {g=true}) {period=3}
  Multiply(
   A=Add(A=Add(A=@n32, B=@n34), B=Multiply(A=@n35) {constB=0.7})
   B=Transform(None=VectorParameter() {group=WPO, parameterName=WPO Direction, defaultValue=<rgba 5 5 0 1>}.RGB) {transformSourceType=TRANSFORMSOURCE_Instance}
  )
  n36 = DynamicParameter() {defaultValue=<rgba 0 1 0 1>, paramNames [UV1 U Offset, Dive, Camera Offset, Param4]}
  n37 = Constant()
  n38 = Panner(
   Coordinate=TextureCoordinate()
   Speed=AppendVector(
    A=ScalarParameter() {group=UV Distortion, parameterName=UV Speed X 2, sliderMax=5, sliderMin=-5}
    B=ScalarParameter() {defaultValue=1, group=UV Distortion, parameterName=UV Speed Y 2, sliderMax=5, sliderMin=-5}
   )
  ) {speedX=-0.5}
  n39 = AppendVector(
   A=ScalarParameter() {defaultValue=1, group=UV Distortion, parameterName=UV Scale X 2, sliderMax=5}
   B=ScalarParameter() {defaultValue=0.25, group=UV Distortion, parameterName=UV Scale Y 2, sliderMax=5}
  )
  n40 = TextureSampleParameter2D(UVs=Add(A=AppendVector(A=@n36.UV1 U Offset, B=@n37), B=Multiply(A=@n38, B=@n39))) {
   group=UV Distortion
   parameterName=Distortion Texture
   texture=/BRCosmetics/Athena/Effects/Particles/Skydiving/NS_Contrail_PinkSpike/Textures/T_PinkSpike_Voronoi2D_Height
  }
  n41 = DynamicParameter() {defaultValue=<rgba 0 1 0 1>, paramNames [UV1 U Offset, Dive, Camera Offset, Param4]}
  n42 = Constant()
  n43 = Panner(
   Coordinate=TextureCoordinate()
   Speed=AppendVector(
    A=ScalarParameter() {group=UV Distortion, parameterName=UV Speed X, sliderMax=5, sliderMin=-5}
    B=ScalarParameter() {defaultValue=2, group=UV Distortion, parameterName=UV Speed Y, sliderMax=5, sliderMin=-5}
   )
  ) {speedX=-0.5}
  n44 = AppendVector(
   A=ScalarParameter() {defaultValue=1, group=UV Distortion, parameterName=UV Scale X, sliderMax=5}
   B=ScalarParameter() {defaultValue=0.5, group=UV Distortion, parameterName=UV Scale Y, sliderMax=5}
  )
  n45 = TextureSampleParameter2D(UVs=Add(A=AppendVector(A=@n41.UV1 U Offset, B=@n42), B=Multiply(A=@n43, B=@n44))) {
   group=UV Distortion
   parameterName=Distortion Texture
   texture=&T_PinkSpike_Voronoi2D_Height
  }
  n46 = Multiply(A=@n40.R, B=@n45.R)
  n47 = Multiply(A=@n40.G, B=@n45.G)
  NamedRerouteDeclaration(None=Subtract(A=MaterialFunctionCall(X=@n46, Y=@n47) {materialFunction=&MakeFloat2}.Result) {constB=0.25}) {name=Distortion, nodeColor=<rgba 1 0.3281 0 1>}
  Normalize(VectorInput=@n2)
  TwoSidedSign()
 ]
}
