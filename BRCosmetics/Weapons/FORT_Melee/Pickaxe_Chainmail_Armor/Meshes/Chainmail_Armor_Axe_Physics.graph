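# Physics asset for the Chainmail Armor pickaxe mesh, defining collision and joints for its dangling chain.
# A capsule on the two-handed handle bone anchors a serial chain of seven dyn_chain bones, each with a hand-fitted capsule body.
# Each link is constrained to the one before it with identical tight swing and twist limits, so the chain bends only slightly at each joint.
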PhysicsAsset /BRCosmetics/Weapons/FORT_Melee/Pickaxe_Chainmail_Armor/Meshes/Chainmail_Armor_Axe_Physics
bodies {
 Melee_TwoHanded_Handle [{
   shapeType=Capsule
   length=6.5274
   radius=3.8672
   shapeName=Melee_TwoHanded_Handle_capsule
   center=<0.6626 -0.3801 -25.3504>
   rotation=<rot -0.093 -38.347 -7.0615>
  }]
 dyn_chain1 [{shapeType=Capsule, length=1, radius=0.82, shapeName=dyn_chain1_capsule, center=<1 0 0>, rotation=<rot 90 0 0>}]
 dyn_chain2 [{shapeType=Capsule, length=3.3232, radius=1.4964, shapeName=dyn_chain2_capsule, center=<3.2 0.9 -0.7>, rotation=<rot -40 88 -88>}]
 dyn_chain3 [{shapeType=Capsule, length=4, radius=3.1, shapeName=dyn_chain3_capsule, center=<2.89 2.5 -0.35>, rotation=<rot 90 0 0>}]
 dyn_chain4 [{shapeType=Capsule, length=3.89, radius=1.55, shapeName=dyn_chain4_capsule, center=<2.07 0.65 0>, rotation=<rot 90 0 0>}]
 dyn_chain5 [{shapeType=Capsule, length=3.2, radius=1.58, shapeName=dyn_chain5_capsule, center=<2.29 0.1 0>, rotation=<rot 90 0 0>}]
 dyn_chain6 [{shapeType=Capsule, length=2, radius=1.5, shapeName=dyn_chain6_capsule, center=<1.6156 -0.7 0>, rotation=<rot 90 0 0>}]
 dyn_chain7 [{shapeType=Capsule, length=2.3, radius=2.1186, shapeName=dyn_chain7_capsule, center=<5.2 0 0>, rotation=<rot 90 0 0>}]
}
constraints [
 {bone1Name=dyn_chain1, bone2Name=Melee_TwoHanded_Handle, swing1LimitDegrees=20, swing2LimitDegrees=20, twistLimitDegrees=10}
 {bone1Name=dyn_chain4, bone2Name=dyn_chain3, swing1LimitDegrees=20, swing2LimitDegrees=20, twistLimitDegrees=10}
 {bone1Name=dyn_chain5, bone2Name=dyn_chain4, swing1LimitDegrees=20, swing2LimitDegrees=20, twistLimitDegrees=10}
 {bone1Name=dyn_chain6, bone2Name=dyn_chain5, swing1LimitDegrees=20, swing2LimitDegrees=20, twistLimitDegrees=10}
 {bone1Name=dyn_chain7, bone2Name=dyn_chain6, swing1LimitDegrees=20, swing2LimitDegrees=20, twistLimitDegrees=10}
 {bone1Name=dyn_chain3, bone2Name=dyn_chain2, swing1LimitDegrees=20, swing2LimitDegrees=20, twistLimitDegrees=10}
 {bone1Name=dyn_chain2, bone2Name=dyn_chain1, swing1LimitDegrees=20, swing2LimitDegrees=20, twistLimitDegrees=10}
]
